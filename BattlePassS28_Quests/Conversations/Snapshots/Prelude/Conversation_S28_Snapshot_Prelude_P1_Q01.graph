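# Conversation database for the Season 28 Prelude snapshot quest P1 Q01, where the NPC Odyssey introduces herself and recruits the player to reconstruct The Mosaic.
# The entry tag leads into a linear chain of four speech nodes.
# The first choice auto-selects as main questline, gated on the NPC's RankedOlympus conversation tag and the quest being active with its objective incomplete.
# The final Continue choice carries a quest-specific tag, likely for objective tracking.
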ConversationDatabase /BattlePassS28_Quests/Conversations/Snapshots/Prelude/Conversation_S28_Snapshot_Prelude_P1_Q01
entry_points [{tag=Conversation.Entry.Season28.RankedOlympus.P1.Q01, destinations [#1]}]
nodes [
 ConversationEntryPointNode_0 {
  id=#1
  class=ConversationEntryPointNode
  outputs [#2]
  properties {entryTag {tagName=Conversation.Entry.Season28.RankedOlympus.P1.Q01}}
 }
 Speech_C_1 {
  id=#2
  class=Speech_C
  outputs [#3]
  properties {
   softTaskIcon=&T_UI_NPC_MainQuest
   generalConfig {defaultMessage=I am Odyssey. A traveler with a keen interest in ancient artifacts.}
  }
  sub_nodes [
   ConversationChoiceNode_2 {
    class=ConversationChoiceNode
    properties {
     defaultChoiceDisplayText=Talk to Odyssey
     choiceTags {gameplayTags [{tagName=Conversation.Choice.AutoSelectIfValid}, {tagName=Conversation.Choice.Category.MainQuestline}]}
    }
   }
   FortConversationRequirement_ControllerRequirement_1 {
    class=FortConversationRequirement_ControllerRequirement
    properties {
     nodeName=Require: NPC Has Conversation Tag
     participantId {tagName=Conversation.Participant.NPC}
     requirement {class=FortControllerRequirement_NPCHasConversationEntryTag, properties {queryTag {tagName=Conversation.Entry.RankedOlympus}}}
    }
   }
   FortConversationRequirement_ControllerRequirement_3 {
    class=FortConversationRequirement_ControllerRequirement
    properties {
     nodeName=Require: Is Quest Active
     requirement {
      class=FortControllerRequirement_IsQuestActive
      properties {questDefinition=/BattlePassS28_Quests/Items/QuestItems/Prelude/Prelude_SnapShot_Quests/Quest_S28_Prelude_Snapshot_P1_Q01}
     }
    }
   }
   FortConversationRequirement_ControllerRequirement_5 {
    class=FortConversationRequirement_ControllerRequirement
    properties {
     nodeName=Require: Has Completed Quest Objective
     requirement {
      class=FortControllerRequirement_HasCompletedQuestObjective
      properties {
       bInverseCondition=true
       objectiveBackendName=Quest_S28_Prelude_Snapshot_P1_Q01_obj0
       questDefinition=&Quest_S28_Prelude_Snapshot_P1_Q01
      }
     }
    }
   }
  ]
 }
 Speech_C_28 {
  id=#3
  class=Speech_C
  outputs [#4]
  properties {generalConfig {defaultMessage=I seek a collaborator to aid in my research.}}
  sub_nodes [ConversationChoiceNode_0 {class=ConversationChoiceNode, properties {defaultChoiceDisplayText=Continue}}]
 }
 Speech_C_2 {
  id=#4
  class=Speech_C
  outputs [#5]
  properties {generalConfig {defaultMessage=Your reputation precedes you as capable and unafraid of risk.}}
  sub_nodes [ConversationChoiceNode_3 {class=ConversationChoiceNode, properties {defaultChoiceDisplayText=Continue}}]
 }
 Speech_C_0 {
  id=#5
  class=Speech_C
  outputs []
  properties {generalConfig {defaultMessage="Help me reconstruct The Mosaic, and we will share in its secrets."}}
  sub_nodes [ConversationChoiceNode_4 {
    class=ConversationChoiceNode
    properties {defaultChoiceDisplayText=Continue, choiceTags {gameplayTags [{tagName=Athena.Quests.S28.Conversation.RankedOlympus.Choice.P1.Q01}]}}
   }]
 }
]
